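# Animation Blueprint for the GreyCar vehicle body, posing axles, wheels, steering and chassis procedurally.
# A Transform Bone chain applies vehicle variables, widens track and wheel wells when bHasTireMod is set, then feeds a bouncy chassis controller.
# An additive state machine layers on top, rolling wheel pivots 90 degrees during boosted airtime and resetting on Land; In Air Start is unlinked.
# EventGraph sets EnableInAir, the alpha for front wheel steering, from ground checks.
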AnimBlueprint /VehicleCosmetics/Bodies/GreyCar/Animation/Body_GreyCar_AnimBP
target_skeleton=/VehicleCosmetics/Bodies/GreyCar/Meshes/SK_Body_GreyCar_Skeleton
graphs [
 AnimGraph {nodes [
   n1 = "Transform (Modify) Bone - Bone: body"(
    ComponentPose=Mesh Space Ref Pose!() -> (ComponentPose:Component Space Pose Link Structure) {is=title AnimGraphNode_MeshRefPose}
    Alpha=1
    Translation="0,0,0"
    Rotation="-0.5,0,0"
    Scale="1,1,1"
   ) {
    rotationMode=BMM_Replace
    rotationSpace=BCS_BoneSpace
    translationMode=BMM_Replace
    translationSpace=BCS_BoneSpace
    alphaBoolBlend {blendOption=Linear}
    boneToModify {boneName=body}
    rotation {pitch=-0.5}
   }
   n2 = "Transform (Modify) Bone - Bone: front_differential"(
    ComponentPose=@n1
    Alpha=1
    Translation=Get AxlePositionF() -> (AxlePositionF:Vector)
    Rotation=Get AxleRotationF() -> (AxleRotationF:Rotator)
    Scale="1,1,1"
   ) {
    rotationMode=BMM_Replace
    rotationSpace=BCS_BoneSpace
    translationMode=BMM_Replace
    translationSpace=BCS_BoneSpace
    alphaBoolBlend {blendOption=Linear}
    boneToModify {boneName=front_differential}
   }
   n3 = "Transform (Modify) Bone - Bone: rear_differential"(
    ComponentPose=@n2
    Alpha=1
    Translation=Get AxlePositionB() -> (AxlePositionB:Vector)
    Rotation=Get AxleRotationB() -> (AxleRotationB:Rotator)
    Scale="1,1,1"
   ) {
    rotationMode=BMM_Replace
    rotationSpace=BCS_BoneSpace
    translationMode=BMM_Replace
    translationSpace=BCS_BoneSpace
    alphaBoolBlend {blendOption=Linear}
    boneToModify {boneName=rear_differential}
   }
   n4 = "Transform (Modify) Bone - Bone: axle_pivot_fr_l"(ComponentPose=@n3, Alpha=1, Translation=Get WheelPositionFL() -> (WheelPositionFL:Vector), Rotation="0,0,0", Scale="1,1,1") {
    rotationMode=BMM_Replace
    rotationSpace=BCS_BoneSpace
    translationMode=BMM_Replace
    translationSpace=BCS_BoneSpace
    alphaBoolBlend {blendOption=Linear}
    boneToModify {boneName=axle_pivot_fr_l}
   }
   n5 = "Transform (Modify) Bone - Bone: axle_pivot_fr_r"(ComponentPose=@n4, Alpha=1, Translation=Get WheelPositionFR() -> (WheelPositionFR:Vector), Rotation="0,0,0", Scale="1,1,1") {
    rotationMode=BMM_Replace
    rotationSpace=BCS_BoneSpace
    translationMode=BMM_Replace
    translationSpace=BCS_BoneSpace
    alphaBoolBlend {blendOption=Linear}
    boneToModify {boneName=axle_pivot_fr_r}
   }
   n6 = "Transform (Modify) Bone - Bone: axle_pivot_bk_r"(ComponentPose=@n5, Alpha=1, Translation=Get WheelPositionBR() -> (WheelPositionBR:Vector), Rotation="0,0,0", Scale="1,1,1") {
    rotationMode=BMM_Replace
    rotationSpace=BCS_BoneSpace
    translationMode=BMM_Replace
    translationSpace=BCS_BoneSpace
    alphaBoolBlend {blendOption=Linear}
    boneToModify {boneName=axle_pivot_bk_r}
   }
   n7 = "Transform (Modify) Bone - Bone: axle_pivot_bk_l"(ComponentPose=@n6, Alpha=1, Translation=Get WheelPositionBL() -> (WheelPositionBL:Vector), Rotation="0,0,0", Scale="1,1,1") {
    rotationMode=BMM_Replace
    rotationSpace=BCS_BoneSpace
    translationMode=BMM_Replace
    translationSpace=BCS_BoneSpace
    alphaBoolBlend {blendOption=Linear}
    boneToModify {boneName=axle_pivot_bk_l}
   }
   n8 = Get EnableInAir() -> (EnableInAir:double)
   n9 = "Transform (Modify) Bone - Bone: wheel_steering_fr_r"(ComponentPose=@n7, Alpha=@n8, Translation="0,0,0", Rotation=Get WheelRotationFR() -> (WheelRotationFR:Rotator), Scale="1,1,1") {
    rotationMode=BMM_Replace
    rotationSpace=BCS_BoneSpace
    translationSpace=BCS_BoneSpace
    alphaBoolBlend {blendOption=Linear}
    boneToModify {boneName=wheel_steering_fr_r}
   }
   n10 = "Transform (Modify) Bone - Bone: wheel_steering_fr_l"(ComponentPose=@n9, Alpha=@n8, Translation="0,0,0", Rotation=Get WheelRotationFL() -> (WheelRotationFL:Rotator), Scale="1,1,1") {
    rotationMode=BMM_Replace
    rotationSpace=BCS_BoneSpace
    translationSpace=BCS_BoneSpace
    alphaBoolBlend {blendOption=Linear}
    boneToModify {boneName=wheel_steering_fr_l}
   }
   n11 = "Transform (Modify) Bone - Bone: wheel_disc_fr_r"(ComponentPose=@n10, Alpha=1, Translation="0,0,0", Rotation=Get WheelSpinFR() -> (WheelSpinFR:Rotator), Scale="1,1,1") {rotationMode=BMM_Replace, rotationSpace=BCS_BoneSpace, alphaBoolBlend {blendOption=Linear}, boneToModify {boneName=wheel_disc_fr_r}}
   n12 = "Transform (Modify) Bone - Bone: wheel_disc_fr_l"(ComponentPose=@n11, Alpha=1, Translation="0,0,0", Rotation=Get WheelSpinFL() -> (WheelSpinFL:Rotator), Scale="1,1,1") {rotationMode=BMM_Replace, rotationSpace=BCS_BoneSpace, alphaBoolBlend {blendOption=Linear}, boneToModify {boneName=wheel_disc_fr_l}}
   n13 = "Transform (Modify) Bone - Bone: wheel_disc_bk_r"(ComponentPose=@n12, Alpha=1, Translation="0,0,0", Rotation=Get WheelSpinBR() -> (WheelSpinBR:Rotator), Scale="1,1,1") {rotationMode=BMM_Replace, rotationSpace=BCS_BoneSpace, alphaBoolBlend {blendOption=Linear}, boneToModify {boneName=wheel_disc_bk_r}}
   n14 = "Transform (Modify) Bone - Bone: wheel_disc_bk_l"(ComponentPose=@n13, Alpha=1, Translation="0,0,0", Rotation=Get WheelSpinBL() -> (WheelSpinBL:Rotator), Scale="1,1,1") {rotationMode=BMM_Replace, rotationSpace=BCS_BoneSpace, alphaBoolBlend {blendOption=Linear}, boneToModify {boneName=wheel_disc_bk_l}}
   n15 = "Transform (Modify) Bone - Bone: steering_wheel"(ComponentPose=@n14, Alpha=1, Translation="0,0,0", Rotation=Get SteeringRotation() -> (SteeringRotation:Rotator), Scale="1,1,1") {rotationMode=BMM_Replace, rotationSpace=BCS_BoneSpace, alphaBoolBlend {blendOption=Linear}, boneToModify {boneName=steering_wheel}}
   n16 = Get bHasTireMod() -> (bHasTireMod:bool)
   n17 = "Transform (Modify) Bone - Bone: axle_pivot_bk_l"!(ComponentPose=@n15, bAlphaBoolEnabled:bool=@n16, Translation="0,-20,0", Rotation="0,0,0", Scale="1,1,1") -> (Pose) {
    alphaInputType=Bool
    rotationSpace=BCS_BoneSpace
    scaleSpace=BCS_BoneSpace
    translationMode=BMM_Additive
    translationSpace=BCS_BoneSpace
    alphaBoolBlend {blendOption=Linear}
    boneToModify {boneName=axle_pivot_bk_l}
    translation {y=-20}
   }
   n18 = "Transform (Modify) Bone - Bone: axle_pivot_bk_r"!(ComponentPose=@n17, bAlphaBoolEnabled:bool=@n16, Translation="0,20,0", Rotation="0,0,0", Scale="1,1,1") -> (Pose) {
    alphaInputType=Bool
    rotationSpace=BCS_BoneSpace
    scaleSpace=BCS_BoneSpace
    translationMode=BMM_Additive
    translationSpace=BCS_BoneSpace
    alphaBoolBlend {blendOption=Linear}
    boneToModify {boneName=axle_pivot_bk_r}
    translation {y=20}
   }
   n19 = "Transform (Modify) Bone - Bone: axle_pivot_fr_l"!(ComponentPose=@n18, bAlphaBoolEnabled:bool=@n16, Translation="0,-18,0", Rotation="0,0,0", Scale="1,1,1") -> (Pose) {
    alphaInputType=Bool
    rotationSpace=BCS_BoneSpace
    translationMode=BMM_Additive
    translationSpace=BCS_BoneSpace
    alphaBoolBlend {blendOption=Linear}
    boneToModify {boneName=axle_pivot_fr_l}
    translation {y=-18}
   }
   n20 = "Transform (Modify) Bone - Bone: axle_pivot_fr_r"!(ComponentPose=@n19, bAlphaBoolEnabled:bool=@n16, Translation="0,18,0", Rotation="0,0,0", Scale="1,1,1") -> (Pose) {
    alphaInputType=Bool
    rotationSpace=BCS_BoneSpace
    translationMode=BMM_Additive
    translationSpace=BCS_BoneSpace
    alphaBoolBlend {blendOption=Linear}
    boneToModify {boneName=axle_pivot_fr_r}
    translation {y=18}
   }
   n21 = "Transform (Modify) Bone - Bone: wheel_well_spreader_bk_l"!(ComponentPose=@n20, bAlphaBoolEnabled:bool=@n16, Translation="3,0,0", Rotation="0,0,0", Scale="1.25,1,1") -> (Pose) {
    alphaInputType=Bool
    rotationSpace=BCS_BoneSpace
    scaleMode=BMM_Additive
    scaleSpace=BCS_BoneSpace
    translationMode=BMM_Additive
    translationSpace=BCS_BoneSpace
    alphaBoolBlend {blendOption=Linear}
    alphaScaleBiasClamp {bInterpResult=true, interpSpeedDecreasing=5, interpSpeedIncreasing=5}
    boneToModify {boneName=wheel_well_spreader_bk_l}
    scale {x=1.25}
    translation {x=3}
   }
   n22 = "Transform (Modify) Bone - Bone: wheel_well_spreader_bk_r"!(ComponentPose=@n21, bAlphaBoolEnabled:bool=@n16, Translation="3,0,0", Rotation="0,0,0", Scale="1.25,1,1") -> (Pose) {
    alphaInputType=Bool
    rotationSpace=BCS_BoneSpace
    scaleMode=BMM_Additive
    scaleSpace=BCS_BoneSpace
    translationMode=BMM_Additive
    translationSpace=BCS_BoneSpace
    alphaBoolBlend {blendOption=Linear}
    alphaScaleBiasClamp {bInterpResult=true, interpSpeedDecreasing=5, interpSpeedIncreasing=5}
    boneToModify {boneName=wheel_well_spreader_bk_r}
    scale {x=1.25}
    translation {x=3}
   }
   n23 = "Transform (Modify) Bone - Bone: wheel_well_spreader_fr_l"!(ComponentPose=@n22, bAlphaBoolEnabled:bool=@n16, Translation="0,0,0", Rotation="0,0,0", Scale="1.25,1,1") -> (Pose) {
    alphaInputType=Bool
    rotationSpace=BCS_BoneSpace
    scaleMode=BMM_Additive
    scaleSpace=BCS_BoneSpace
    translationMode=BMM_Additive
    translationSpace=BCS_BoneSpace
    alphaBoolBlend {blendOption=Linear}
    alphaScaleBiasClamp {bInterpResult=true, interpSpeedDecreasing=5, interpSpeedIncreasing=5}
    boneToModify {boneName=wheel_well_spreader_fr_l}
    scale {x=1.25}
   }
   n24 = "Transform (Modify) Bone - Bone: wheel_well_spreader_fr_r"!(ComponentPose=@n23, bAlphaBoolEnabled:bool=@n16, Translation="0,0,0", Rotation="0,0,0", Scale="1.25,1,1") -> (Pose) {
    alphaInputType=Bool
    rotationSpace=BCS_BoneSpace
    scaleMode=BMM_Additive
    scaleSpace=BCS_BoneSpace
    translationMode=BMM_Additive
    translationSpace=BCS_BoneSpace
    alphaBoolBlend {blendOption=Linear}
    alphaScaleBiasClamp {bInterpResult=true, interpSpeedDecreasing=5, interpSpeedIncreasing=5}
    boneToModify {boneName=wheel_well_spreader_fr_r}
    scale {x=1.25}
   }
   AnimGraphNode_BouncyChassisVehicleController_1 = Vehicle Bouncy Chassis Controller!(
    ComponentPose:Component Space Pose Link Structure=@n24
    Alpha:float=1
    BouncyChassis:Bouncy Chassis Vehicle State Structure=(PitchDegrees=0,RollDegrees=0,TranslationZ=0)
   ) -> (Pose:Component Space Pose Link Structure) {is=title AnimGraphNode_VehicleBouncyChassisController, alphaBoolBlend {blendOption=Linear}}
   AnimGraphNode_StateMachine_1 = GreyCar State Machine() {is=title AnimGraphNode_StateMachine}
   Output Pose(Result=Apply Additive(Base=Component To Local(ComponentPose=@AnimGraphNode_BouncyChassisVehicleController_1), Additive=@AnimGraphNode_StateMachine_1) {alphaBoolBlend {blendOption=Linear}})
   Get EnableInAir()
  ]}
 GreyCar State Machine {
  entry=On Ground
  path=AnimGraph.AnimGraphNode_StateMachine_1.GreyCar State Machine
  ids=id
  nodes [
   AnimStateNode_0 = On Ground() {is=title AnimStateNode}
   AnimStateNode_1 = In Air Start() {is=title AnimStateNode}
   AnimStateNode_2 = In Air Loop() {is=title AnimStateNode}
   AnimStateNode_3 = Land() {is=title AnimStateNode}
   AnimStateTransitionNode_1 = AnimStateTransitionNode() {bAllowInertializationForSelfTransitions=true}
   AnimStateTransitionNode_6 = AnimStateTransitionNode() {crossfadeDuration=0.1}
   AnimStateTransitionNode_7 = AnimStateTransitionNode() {crossfadeDuration=0.1}
  ]
  links [
   {
    dst=Land
    rule=AnimGraph.AnimGraphNode_StateMachine_1.GreyCar State Machine.AnimStateTransitionNode_6.AnimationTransitionGraph_0
    src=In Air Loop
    via=AnimStateTransitionNode_6
   }
   {
    dst=On Ground
    rule=AnimGraph.AnimGraphNode_StateMachine_1.GreyCar State Machine.AnimStateTransitionNode_7.AnimationTransitionGraph_0
    src=Land
    via=AnimStateTransitionNode_7
   }
   {
    dst=In Air Loop
    rule=AnimGraph.AnimGraphNode_StateMachine_1.GreyCar State Machine.AnimStateTransitionNode_1.Transition
    src=On Ground
    via=AnimStateTransitionNode_1
   }
  ]
 }
 On Ground {
  path=AnimGraph.AnimGraphNode_StateMachine_1.GreyCar State Machine.AnimStateNode_0.On Ground
  nodes [
   Sequence Player()
   Output Animation Pose(Result="Sequence Player (additive) 'Greycar_NO_ANIM'"() {bLoopAnimation=false, sequence=&/Greycar_NO_ANIM})
  ]
 }
 In Air Start {
  path=AnimGraph.AnimGraphNode_StateMachine_1.GreyCar State Machine.AnimStateNode_1.In Air Start
  nodes [Output Animation Pose(Result="Sequence Player (additive) 'Greycar_InAir_Start'"() {bLoopAnimation=false, sequence=&/Greycar_InAir_Start})]
 }
 In Air Loop {
  path=AnimGraph.AnimGraphNode_StateMachine_1.GreyCar State Machine.AnimStateNode_2.In Air Loop
  nodes [
   n1 = Local To Component(LocalPose="Sequence Player (additive) 'Greycar_NO_ANIM'"() {sequence=&Greycar_NO_ANIM})
   n2 = "Transform (Modify) Bone - Bone: steering_knuckle_fr_l"(
    ComponentPose="Transform (Modify) Bone - Bone: axle_pivot_bk_r"!(
     ComponentPose="Transform (Modify) Bone - Bone: axle_pivot_bk_l"!(ComponentPose=@n1, Alpha:float=1, Translation="0,0,0", Rotation="0,0,-90", Scale="1,1,1") -> (Pose) {rotationMode=BMM_Replace, boneToModify {boneName=axle_pivot_bk_l}, rotation {roll=-90}}
     Alpha:float=1
     Translation="0,0,0"
     Rotation="0,0,90"
     Scale="1,1,1"
    ) -> (Pose) {rotationMode=BMM_Replace, boneToModify {boneName=axle_pivot_bk_r}, rotation {roll=90}}
    Alpha=1
    Translation="0,0,0"
    Rotation="0,0,-90"
    Scale="1,1,1"
   ) {rotationMode=BMM_Replace, boneToModify {boneName=steering_knuckle_fr_l}, rotation {roll=-90}}
   Output Animation Pose(Result=Component To Local(ComponentPose="Transform (Modify) Bone - Bone: steering_knuckle_fr_r"(ComponentPose=@n2, Alpha=1, Translation="0,0,0", Rotation="0,0,90", Scale="1,1,1") {rotationMode=BMM_Replace, boneToModify {boneName=steering_knuckle_fr_r}, rotation {roll=90}})) {
    node {stateEntryFunction {functionName=OnStateEntry}}
    stateEntryFunction {bSelfContext=true, memberGuid=#1, memberName=OnStateEntry_Air}
   }
  ]
 }
 Land {
  path=AnimGraph.AnimGraphNode_StateMachine_1.GreyCar State Machine.AnimStateNode_3.Land
  nodes [
   n1 = Local To Component(LocalPose="Sequence Player (additive) 'Greycar_NO_ANIM'"() {bLoopAnimation=false, sequence=&Greycar_NO_ANIM})
   n2 = "Transform (Modify) Bone - Bone: steering_knuckle_fr_l"(
    ComponentPose="Transform (Modify) Bone - Bone: axle_pivot_bk_r"(
     ComponentPose="Transform (Modify) Bone - Bone: axle_pivot_bk_l"(ComponentPose=@n1, Alpha=1, Translation="0,0,0", Rotation="0,0,0", Scale="1,1,1") {rotationMode=BMM_Replace, boneToModify {boneName=axle_pivot_bk_l}}
     Alpha=1
     Translation="0,0,0"
     Rotation="0,0,0"
     Scale="1,1,1"
    ) {rotationMode=BMM_Replace, boneToModify {boneName=axle_pivot_bk_r}}
    Alpha=1
    Translation="0,0,0"
    Rotation="0,0,0"
    Scale="1,1,1"
   ) {rotationMode=BMM_Replace, boneToModify {boneName=steering_knuckle_fr_l}}
   Output Animation Pose(Result=Component To Local(ComponentPose="Transform (Modify) Bone - Bone: steering_knuckle_fr_r"(ComponentPose=@n2, Alpha=1, Translation="0,0,0", Rotation="0,0,0", Scale="1,1,1") {rotationMode=BMM_Replace, boneToModify {boneName=steering_knuckle_fr_r}})) {
    node {stateEntryFunction {functionName=OnStateEntry_0}}
    stateEntryFunction {bSelfContext=true, memberGuid=#2, memberName=OnStateEntry_0}
   }
  ]
 }
 Transition {
  path=AnimGraph.AnimGraphNode_StateMachine_1.GreyCar State Machine.AnimStateTransitionNode_1.Transition
  nodes [
   n1 = NOT Boolean(A=Get bIsTouchingGround() -> (bIsTouchingGround:bool))
   n2 = "float >= float"(A:bool=Get RootBoneLocation() -> (RootBoneLocation:bool), B=Get DistanceToGround() -> (DistanceToGround:float))
   Result(bCanEnterTransition=AND Boolean(A=AND Boolean(A=@n2, B=@n1), B=Get bIsBoosting() -> (bIsBoosting:bool)))
  ]
 }
 AnimationTransitionGraph_0 {
  path=AnimGraph.AnimGraphNode_StateMachine_1.GreyCar State Machine.AnimStateTransitionNode_6.AnimationTransitionGraph_0
  nodes [Result(bCanEnterTransition="float <= float"(A:bool=Get RootBoneLocation(), B=Get DistanceToGround()) -> (ReturnValue=~))]
 }
 AnimationTransitionGraph_0 {
  path=AnimGraph.AnimGraphNode_StateMachine_1.GreyCar State Machine.AnimStateTransitionNode_7.AnimationTransitionGraph_0
  nodes [Result(bCanEnterTransition="float <= float"(A:double="Get Relevant Anim Time Remaining (Land)"!() -> (ReturnValue:float) {is=title K2Node_AnimGetter}, B=0.1) -> (ReturnValue=~))]
 }
 EventGraph {nodes [
   Event BlueprintUpdateAnimation()
   > Sequence()
   n1 = Get bIsTouchingGround()
   n2 = Get RootBoneLocation()
   n3 = Get DistanceToGround()
   > n4 = Branch(Condition=AND Boolean(A=NOT Boolean(A=@n1), B="float >= float"(A=@n2, B=@n3)))
   > Set EnableInAir(EnableInAir:double=0) -> (Output_Get:double)
   Set EnableInAir(execute=@n4.else, EnableInAir=1)
   TryGetPawnOwner()
   n5 = GetBoneTransform!(self:Skinned Mesh Component Object Reference, InBoneName:Name=root, TransformSpace:ERelativeTransformSpace Enum=RTS_World) -> (ReturnValue:Transform)
   Set RootBoneLocation(RootBoneLocation:bool="Equal (Vector)"!(A:Vector=BreakTransform(InTransform=@n5), B:Vector, ErrorTolerance:float) -> (ReturnValue:bool=~) {is=title K2Node_PromotableOperator}) -> (Output_Get:bool)
  ]}
 OnStateEntry_Air {nodes [
   OnStateEntry_Air!() -> (then, Context:"Anim Update Context Structure (by ref)", Node:"Anim Node Reference Structure (by ref)") {is=title K2Node_FunctionEntry}
   > Set EnableWheelBost(EnableWheelBost:bool=true) -> (Output_Get:bool)
   > Return Node()
  ]}
 OnStateEntry_Landing {nodes [
   OnStateEntry_Landing!() -> (then, Context:"Anim Update Context Structure (by ref)", Node:"Anim Node Reference Structure (by ref)") {is=title K2Node_FunctionEntry}
   > Set EnableWheelBost(EnableWheelBost=false)
   > Return Node()
  ]}
]
variables [
 EliminateSteering {type=boolean}
 EnableInAir {type=number}
 EnableWheelBost {type=boolean}
 RootBoneLocation {type=boolean}
]
