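; Ambient cosmetic FX for the GreyCar vehicle body, driving an attached particle effect and temporary flying start/stop sounds.
; BeginPlay attaches the effect to the body mesh root, caches its AnimBP and starts a looping GroundCheck timer.
; GroundCheck polls the AnimBP's wheel-boost flag to set bIsFlying and trigger DoOnce-gated sounds.
; Boost events toggle bIsBoosting, with starts gated by significance LOD; EndPlay deactivates and detaches the effect.
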(event EventBeginPlay
  (bind _vehiclecosmeticsinterface (Variables|Vehicle|Cosmetics|GetVehicleCosmeticsInterface))
  (bind _returnvalue (Vehicle|Cosmetics|GetBodySkeletalMeshComponent _vehiclecosmeticsinterface))
  (bind _ambient_vfx (Variables|Default|GetAmbient%20VFX))
  (bind _asbody_grey_car_anim_bp (Utilities|Casting|CastToBody_GreyCar_AnimBP (Components|SkeletalMesh|GetAnimInstance _returnvalue))
    (:then
      (Variables|Default|SetAs%20Body%20Grey%20Car%20Anim%20BP _asbody_grey_car_anim_bp))
    (:CastFailed
      _))
  (Transformation|AttachComponentToComponent _ambient_vfx _returnvalue "root" "KeepRelative" "KeepRelative" "KeepRelative" true)
  (Components|Activation|Activate _ambient_vfx true)
  (Utilities|Time|SetTimerbyFunctionName 0 "GroundCheck" 0.2 true))

(event Collision|EventActorBeginOverlap
  (params (OtherActor :type Actor%20Object%20Reference)))

(event EventTick
  (params (DeltaSeconds :type Float%20%28single-precision%29)))

(event Custom|GroundCheck
  (bind _as_body_grey_car_anim_bp (Variables|Default|GetAs%20Body%20Grey%20Car%20Anim%20BP))
  (Utilities|IsValid _as_body_grey_car_anim_bp
    (:Is%20Valid
      (Effects|Components|ParticleSystem|SetBoolParameter (Variables|Default|GetAmbient%20VFX) "bIsFlying" (Variables|/VehicleCosmetics/Bodies/GreyCar/Animation/Body_GreyCar_AnimBP.Body_GreyCar_AnimBP_C|GetEnableWheelBost _as_body_grey_car_anim_bp))
      ;[0.3333 0 0.03039 0.2] Temp: Trigger sound based on wheel animation state. This behavior should be moved to a new vehicle_sound_component which will generically update engine metasounds with 'DistanceToGround' float input.
      (if (!= (Variables|Default|GetPreviousEnableWheelBoost) (Variables|/VehicleCosmetics/Bodies/GreyCar/Animation/Body_GreyCar_AnimBP.Body_GreyCar_AnimBP_C|GetEnableWheelBost (Variables|Default|GetAs%20Body%20Grey%20Car%20Anim%20BP)))
        (if (Variables|/VehicleCosmetics/Bodies/GreyCar/Animation/Body_GreyCar_AnimBP.Body_GreyCar_AnimBP_C|GetEnableWheelBost (Variables|Default|GetAs%20Body%20Grey%20Car%20Anim%20BP))
          (Utilities|FlowControl|Sequence
            (:then_0
              (Utilities|FlowControl|DoOnce false :!label _label
                (:Completed
                  (Audio|SpawnSoundAttached (Variables|Default|GetFlyingStartSound) (Vehicle|Cosmetics|GetBodySkeletalMeshComponent (Variables|Vehicle|Cosmetics|GetVehicleCosmeticsInterface))))))
            (:then_1
              (Utilities|FlowControl|DoOnce true :!label _label_2
                (:Completed
                  (Audio|SpawnSoundAttached (Variables|Default|GetFlyingStopSound) (Vehicle|Cosmetics|GetBodySkeletalMeshComponent (Variables|Vehicle|Cosmetics|GetVehicleCosmeticsInterface)))))))
          (Variables|Default|SetPreviousEnableWheelBoost (Variables|/VehicleCosmetics/Bodies/GreyCar/Animation/Body_GreyCar_AnimBP.Body_GreyCar_AnimBP_C|GetEnableWheelBost (Variables|Default|GetAs%20Body%20Grey%20Car%20Anim%20BP)))
          (else
            (Utilities|FlowControl|Sequence
              (:then_0
                _)
              (:then_1
                (goto _label Reset)))))
        (goto _label_2)))))

(event Vehicle|Cosmetics|EventOnBoostStarted
  (bind _returnvalue (VehicleCosmetics|FX|GetMaxSignificanceLODBoostFX))
  (bind _btrigger (Class|BPVehicleCosmeticsLibrary|ShouldTriggerFX (Variables|Vehicle|Cosmetics|GetVehicleCosmeticsInterface) _returnvalue))
  (if _btrigger
    (Effects|Components|ParticleSystem|SetBoolParameter (Variables|Default|GetAmbient%20VFX) "bIsBoosting" true)))

(event Vehicle|Cosmetics|EventOnBoostFinished
  (Effects|Components|ParticleSystem|SetBoolParameter (Variables|Default|GetAmbient%20VFX) "bIsBoosting" false))

(event EventEndPlay
  (params (EndPlayReason :type EEndPlayReason%20Enum :default "Destroyed"))
  (bind _ambient_vfx (Variables|Default|GetAmbient%20VFX))
  (Components|Activation|Deactivate _ambient_vfx)
  (Transformation|DetachFromComponent _ambient_vfx "KeepRelative" "KeepRelative" "KeepRelative" true)
  (Effects|Components|ParticleSystem|SetBoolParameter _ambient_vfx "bIsBoosting" false))

(fn ConstructionScript)
