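# Niagara system for the GreyCar vehicle body's ambient effects: thruster flames, booster sprites, electric zaps and speed lines.
# System timelines and vehicle velocity and flip state drive spawn rates, scale and fades.
# Flames crossfade two groups on four skeletal-mesh sockets, where GPU booster sprites also spawn downward-aligned.
# Zaps use hardcoded body positions; speed-line flares follow a sampled point-array curve; BoostFlare_LW is disabled and Portal_LW has no renderers.
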NiagaraSystem /VehicleCosmetics/Bodies/GreyCar/FX/NS_Ambient_GreyCar
dependencies {
 usedDataInterfaces [
  &NiagaraDataInterfaceCurve
  &NiagaraDataInterfaceMeshRendererInfo
  &NiagaraDataInterfaceColorCurve
  &NiagaraDataInterfaceArrayFloat3
  &NiagaraDataInterfaceArrayFloat
  &NiagaraDataInterfaceVector2DCurve
  &NiagaraDataInterfaceArrayPosition
  &NiagaraDataInterfaceSkeletalMesh
  &NiagaraDataInterfaceVectorCurve
 ]
 usedDynamicInputs [
  &FN_FrontEndFloatModifier
  &ClampFloat
  &VectorLength
  /Game/Effects/Niagara/DynamicInputs/Math/FN_RemapDotProduct
  &SubtractVectorFromPosition
  &Multiply_LinearColorByFloat
  &VectorFromCurve
 ]
 usedModules [
  &self:SystemSpawnScript.SystemScriptSource.SystemScriptGraph.NiagaraNodeAssignment_0.SetVariables_#1
  &self:SystemSpawnScript.SystemScriptSource.SystemScriptGraph.NiagaraNodeAssignment_1.SetVariables_#2
  &SystemState
  /Game/Effects/Niagara/Modules/Vehicle/FN_UpdateVehicleAttribs
  &Timeline
  &self:SystemSpawnScript.SystemScriptSource.SystemScriptGraph.NiagaraNodeAssignment_3.SetVariables_#3
  &InterpolateOverTime
  &self:SystemSpawnScript.SystemScriptSource.SystemScriptGraph.NiagaraNodeAssignment_2.SetVariables_#4
  &self:SystemSpawnScript.SystemScriptSource.SystemScriptGraph.NiagaraNodeAssignment_4.SetVariables_#5
  &self:SystemSpawnScript.SystemScriptSource.SystemScriptGraph.NiagaraNodeAssignment_5.SetVariables_#6
 ]
}
emitters {
 BoostFlare_LW {bEnabled=false, rendererClasses [], renderers []}
 Booster001 {
  simTarget=GPUComputeSim
  emitterUpdateScript [
   &EmitterState(Inactive Response=::NewEnumerator1, Life Cycle Mode=::NewEnumerator1, Loop Duration=1, UseLoopDelay=false)
   &SpawnRate(
    SpawnRate=&FloatFromCurve(
     FloatCurve=DataInterface() {propertyValues {Curve {defaultValue=1, keys [<at 0 0 RCIM_Cubic>, {tangentMode=RCTM_User, time=0.3, value=1}]}}}
     CurveIndex=$System.Timeline001.NormalizedTime
     Scale Curve=&Multiply_Float(
      A=&FloatFromCurve(
       FloatCurve=DataInterface() {propertyValues {Curve {
          defaultValue=1
          keys [
           <at 0 0 RCIM_Cubic>
           {arriveTangent=1, interpMode=RCIM_Cubic, leaveTangent=1, time=0.0671, value=0.6835}
           {tangentMode=RCTM_User, time=1, value=1}
          ]
         }}}
       CurveIndex=$System.NormalizedVelocityActivation
       Scale Curve=400
      )
      B=$System.IsNotFlipped
     )
    )
    Use Spawn Probability=false
   )
  ]
  particleSpawnScript [
   &PartitionParticles(Number Of Partitions=4)
   &InitializeParticle(
    Color=<rgba 1 0.3757 0.1729 1>
    Lifetime Max=0.5
    Lifetime Min=0.4
    Lifetime Mode=::NewEnumerator1
    Mass Mode=::NewEnumerator3
    Sprite Size Max=<30 120>
    Sprite Size Min=<15 50>
    Sprite Size Mode=::NewEnumerator2
    UsePositionOffset=false
    Write Material Random=true
   )
   &DynamicMaterialParameters(
    Index 0 Param 1=&RandomRangeFloat(Minimum=3, Maximum=4)
    Index 0 Param 2=&RandomRangeFloat(Minimum=0.4, Maximum=1)
    Index 0 Param 3=16
    Index 0 Param 4=&RandomRangeFloat(Minimum=0.55, Maximum=0.75)
    Index 1 Param 1=1804.9612
    Index 1 Param 2=289.7086
    Index 1 Param 3=26.0864
    Module.Index0Float0Write=true
    Module.Index0Float1Write=true
    Module.Index0Float2Write=true
    Module.Index0Float3Write=true
    Module.Index1Float0Write=true
    Module.Index1Float1Write=true
    Module.Index1Float2Write=true
    Module.Index1Float3Write=true
    Write Parameter Index 1=::NewEnumerator1
   )
   &SkeletalMeshLocation(
    Bone / Socket Index=&Add_Integer(A=1, B=$Particles.PartitionParticles.Partition)
    MaskSpawnGroup=false
    Mesh Sampling Type=::NewEnumerator5
    Orientation Sampling=::NewEnumerator1
    Sampled Position Offset=&SelectVectorFromArray(
     Direct Array Index=$Particles.PartitionParticles.Partition
     Direct Array Mode=::NewEnumerator1
     Vector Selection Array=DataInterface() {propertyValues {FloatData [<0 0 -15>, <0 0 -15>, <0 0 -15>, <0 0 -15>]}}
    )
    Sampled Position Offset Coordinate Space:ENiagaraSkelMeshTransforms=::NewEnumerator2
    Skeletal Mesh=$System.SkeletalMesh
    Socket Sampling Mode=::NewEnumerator5
   )
   &ShapeLocation(
    Cylinder Height=5
    Cylinder Radius=25
    MaskSpawnGroup=false
    Override Seed=false
    Shape Primitive=::NewEnumerator1
    Surface Only=false
    UseOwnerScale=true
   )
  ]
  particleUpdateScript [
   &ParticleState()
   &ScaleSpriteSize(
    Non-Uniform Curve Index:NiagaraFloat=$System.Timeline001.NormalizedTime
    Non-Uniform Curve Sprite Scale=DataInterface() {propertyValues {
      XCurve {keys [<at 0 1 RCIM_Cubic>,<at 0.3609 3.5356 RCIM_Cubic>,<at 1 1 RCIM_Cubic>]}
      YCurve {keys [
        <at 0 0 RCIM_Cubic>
        {arriveTangent=1.5594, interpMode=RCIM_Cubic, leaveTangent=1.5594, time=0.3878, value=10}
        {arriveTangent=-14.7011, interpMode=RCIM_Cubic, leaveTangent=-14.7011, time=0.7974, value=1.2434}
        <at 1 1 RCIM_Cubic>
       ]}
     }}
    Scale Sprite Size Mode=::NewEnumerator3
   )
   /SetVariables(Particles.SpriteAlignment=&AddVectorToPosition(Position=<0 0 -1>, Vector=&RandomRangeVector(Minimum=<-0.05 -0.05 -0.05>, Maximum=<0.05 0.05 0.05>))) {isSetParameters=true}
   &ScaleColor(
    Scale Alpha=&FloatFromCurve(
     FloatCurve=DataInterface() {propertyValues {Curve {defaultValue=1, keys [<at 0 0 RCIM_Cubic>, {tangentMode=RCTM_User, time=1, value=1}]}}}
     CurveIndex=$System.Timeline001.NormalizedTime
     Scale Curve=$System.IsNotFlipped
    )
    Scale RGB=&Multiply_VectorByFloat(
     Vector=<1 1 1>
     Float=&FloatFromCurve(
      FloatCurve=DataInterface() {propertyValues {Curve {
         defaultValue=1
         keys [
          <at 0 0 RCIM_Cubic>
          {arriveTangent=2.3823, interpMode=RCIM_Cubic, leaveTangent=2.3823, time=0.1231, value=2}
          {arriveTangent=0.4908, interpMode=RCIM_Cubic, leaveTangent=0.4908, tangentMode=RCTM_User, time=0.2432, value=0.5794}
          {tangentMode=RCTM_User, time=1, value=1}
         ]
        }}}
      CurveIndex=$System.Timeline001.NormalizedTime
      Scale Curve=10
     )
    )
    ScaleA=true
    ScaleRGB=true
   )
   &SolveForcesAndVelocity(Clamp Velocity=false, Limit Acceleration=false)
  ]
  rendererClasses [&NiagaraSpriteRendererProperties]
  renderers [{rendererClass=&NiagaraSpriteRendererProperties}]
 }
 Flames {
  emitterUpdateScript [
   &EmitterState(
    Inactive Response=::NewEnumerator1
    Life Cycle Mode=::NewEnumerator1
    Loop Behavior=::NewEnumerator1
    Loop Duration Mode=::NewEnumerator1
    UseLoopDelay=false
   )
   &SpawnBurst_Instantaneous(Spawn Count=4, Use Spawn Probability=false, UseLoopCountLimit=false)
   &SpawnBurst_Instantaneous(Spawn Count=4, Spawn Group=1, Use Spawn Probability=false, UseLoopCountLimit=false) {module=SpawnBurst_Instantaneous001}
   /SetVariables(Emitter.DynamicInertia:NiagaraFloat=&Lerp_Float(A=0.4, B=0.1, Alpha=$System.Speed_Alpha)) {isSetParameters=true}
  ]
  particleSpawnScript [
   &PartitionParticles(Number Of Partitions=4)
   &InitializeParticle(
    Mesh Scale=&Multiply_VectorByFloat(
     Vector=<1 1 1>
     Float=&MergeHelper_FloatSwitch(
      0=&SelectFloatFromArray(Direct Array Index=$Particles.PartitionParticles.Partition, Float Selection Array=DataInterface() {propertyValues {FloatData [1,1,1.1,1.1]}})
      1=1
      2=?
      3=?
      Index=&SpawnGroup()
      Num=::NewEnumerator0
     )
    )
    Mesh Scale Mode=::NewEnumerator4
    Position Offset=&SelectVectorFromArray(
     Direct Array Index=$Particles.PartitionParticles.Partition
     Vector Selection Array=DataInterface() {propertyValues {FloatData [<140 -90 20>, <1 1 1>, <0 0 0>, <0 0 0>]}}
    )
    Sprite Size=<300 180>
    Sprite Size Mode=::NewEnumerator4
    UsePositionOffset=false
    Write Material Random=true
   )
   &SkeletalMeshLocation(
    Bone / Socket Index=&Add_Integer(A=1, B=$Particles.PartitionParticles.Partition)
    MaskSpawnGroup=false
    Mesh Sampling Type=::NewEnumerator5
    Orientation Sampling=::NewEnumerator1
    Sampled Position Offset=&MergeHelper_VectorSwitch(
     0=&SelectVectorFromArray(
      Direct Array Index=$Particles.PartitionParticles.Partition
      Vector Selection Array=DataInterface() {propertyValues {FloatData [<-50 10 -3>, <-50 -10 -3>, <-60 -5 -10>, <-60 5 -10>]}}
     )
     1=&SelectVectorFromArray(
      Direct Array Index=$Particles.PartitionParticles.Partition
      Direct Array Mode=::NewEnumerator1
      Vector Selection Array=DataInterface() {propertyValues {FloatData [<0 0 -30>, <0 0 -30>, <0 0 -30>, <0 0 -30>]}}
     )
     2=?
     3=?
     Index=&SpawnGroup()
     Num=::NewEnumerator0
    )
    Sampled Position Offset Coordinate Space:ENiagaraSkelMeshTransforms=::NewEnumerator2
    Skeletal Mesh=$System.SkeletalMesh
    Socket Sampling Mode=::NewEnumerator5
   )
   &InitialMeshOrientation(
    Mesh Orientation Mode=::NewEnumerator1
    Rotation=&MergeHelper_VectorSwitch(
     0=&Multiply_VectorByFloat(
      Vector=&SelectVectorFromArray(
       Direct Array Index=$Particles.PartitionParticles.Partition
       Direct Array Mode=::NewEnumerator1
       Vector Selection Array=DataInterface() {propertyValues {FloatData [<0 0 0>, <0 0 0>, <0 0 0.1>, <0 0 -0.06>]}}
      )
      Float=1
     )
     1=<0 0.25 0>
     2=?
     3=?
     Index=&SpawnGroup()
     Num=::NewEnumerator0
    )
    Use Rotation Vector=true
   )
  ]
  particleUpdateScript [
   &ParticleState(Kill Particles When Lifetime Has Elapsed=false, Let Infinitely Lived Particles Die When Emitter Deactivates=true)
   &ScaleSpriteSize(
    Uniform Curve Index=$System.Timeline001.NormalizedTime
    Uniform Curve Sprite Scale=DataInterface() {propertyValues {Curve {keys [
        <at 0 0 RCIM_Cubic>
        {arriveTangent=1, interpMode=RCIM_Cubic, leaveTangent=1, time=0.0318, value=0.7559}
        {tangentMode=RCTM_User, time=1, value=1}
       ]}}}
   )
   &ScaleMeshSize(Scale Factor=&VectorFromFloat(Value=&MergeHelper_FloatSwitch(
      0=&FloatFromCurve(
       FloatCurve=DataInterface() {propertyValues {Curve {defaultValue=1, keys [<at 0 0.2 RCIM_Cubic>, {tangentMode=RCTM_User, time=0.4, value=1}]}}}
       CurveIndex=$System.NormalizedVelocityActivation
       Scale Curve=&FloatFromCurve(FloatCurve=DataInterface() {propertyValues {Curve {keys [<at 0 1 RCIM_Cubic>,<at 1 0>]}}}, CurveIndex=$System.Timeline001.NormalizedTime)
      )
      1=&FloatFromCurve(
       FloatCurve=DataInterface() {propertyValues {Curve {
          defaultValue=1
          keys [
           <at 0 0 RCIM_Cubic>
           {arriveTangent=1.7578, interpMode=RCIM_Cubic, leaveTangent=1.7578, time=0.1644, value=1.4271}
           {arriveTangent=0.2012, interpMode=RCIM_Cubic, leaveTangent=0.2012, tangentMode=RCTM_User, time=0.2485, value=0.4369}
           {tangentMode=RCTM_User, time=1, value=1}
          ]
         }}}
       CurveIndex=$System.Timeline001.NormalizedTime
       Scale Curve=$System.IsNotFlipped
      )
      2=?
      3=?
      Index=&SpawnGroup()
      Num=::NewEnumerator0
     )))
   &DynamicMaterialParameters(
    Index 0 Param 1=&FloatFromCurve(
     FloatCurve=DataInterface() {propertyValues {Curve {defaultValue=1, keys [{interpMode=RCIM_Cubic, tangentMode=RCTM_Break, time=0.2, value=0.3}, {tangentMode=RCTM_User, time=0.7, value=1}]}}}
     CurveIndex=$System.NormalizedVelocityActivation
     Scale Curve=-60
    )
    Index 0 Param 2=&FloatFromCurve(
     FloatCurve=DataInterface() {propertyValues {Curve {defaultValue=1, keys [<at 0.2 0.3 RCIM_Cubic>, {tangentMode=RCTM_User, time=0.7, value=1}]}}}
     CurveIndex=$System.NormalizedVelocityActivation
     Scale Curve=&SelectFloatFromArray(Direct Array Index=&SpawnGroup(), Float Selection Array=DataInterface() {propertyValues {FloatData [-40,-80]}})
    )
    Index 0 Param 3=&Multiply_Float(A=$Engine.Time, B=&RandomRangeFloat(Minimum=-0.6, Maximum=-1))
    Index 0 Param 4=1
    Index 1 Param 1=1
    Index 1 Param 2=1
    Index 1 Param 3=1
    Index 2 Param 1=1
    Module.Index0Float0Write=true
    Module.Index0Float1Write=true
    Module.Index0Float2Write=true
    Module.Index0Float3Write=true
    Module.Index1Float0Write=true
    Module.Index1Float1Write=true
    Module.Index1Float2Write=true
    Module.Index1Float3Write=true
    Module.Index2Float0Write=true
    Module.Index2Float1Write=true
    Module.Index2Float2Write=true
    Module.Index2Float3Write=true
    Write Parameter Index 1=::NewEnumerator1
    Write Parameter Index 2=::NewEnumerator1
   ) {module=DynamicMaterialParameters001}
   &ScaleColor(
    Scale Alpha=&FloatFromCurve(
     FloatCurve=DataInterface() {propertyValues {Curve {keys [<at 0 1 RCIM_Cubic>,<at 1 0>]}}}
     CurveIndex=&MergeHelper_FloatSwitch(
      0=$System.Timeline001.NormalizedTime
      1=&OneMinusFloat(Float=$System.Timeline001.NormalizedTime)
      2=?
      3=?
      Index=&SpawnGroup()
      Num=::NewEnumerator0
     )
     Scale Curve=&FloatFromCurve(
      FloatCurve=DataInterface() {propertyValues {Curve {defaultValue=1, keys [<at 0.4 0 RCIM_Cubic>, {tangentMode=RCTM_User, time=0.6, value=1}]}}}
      CurveIndex=$System.NormalizedVelocityActivation
      Scale Curve=$System.IsNotFlipped
     )
    )
    ScaleA=true
    ScaleRGB=true
   )
   /SetVariables(Particles.FlareColor=&MultiplyLinearColor(
     Linear Color=$Particles.Color
     Scale Factor=&MakeVector4_FromLinearColor(LinearColor=<rgba 1 0.5687 0.3643 1>)
    )) {isSetParameters=true}
  ]
  rendererClasses [&NiagaraMeshRendererProperties, &NiagaraSpriteRendererProperties]
  renderers [
   {rendererClass=&NiagaraMeshRendererProperties, rendererIndex=0}
   {rendererClass=&NiagaraMeshRendererProperties, rendererIndex=1}
   {rendererClass=&NiagaraSpriteRendererProperties, rendererIndex=2}
  ]
 }
 NE_Zaps {
  emitterUpdateScript [
   &EmitterState()
   &SpawnRate(
    SpawnRate=&Waveform(
     Global Amplitude Scale=&FloatFromCurve(
      FloatCurve=DataInterface() {propertyValues {Curve {
         defaultValue=1
         keys [
          <at 0 0 RCIM_Cubic>
          {arriveTangent=3.0192, interpMode=RCIM_Cubic, leaveTangent=3.0192, tangentMode=RCTM_User, time=0.7758, value=0.6265}
          <at 1 1 RCIM_Cubic>
         ]
        }}}
      CurveIndex=$System.Timeline.NormalizedTime
      Scale Curve=50
     )
     "Waveform [1]"=::NewEnumerator10
     "[1] Frequency"=4
     "[1] Amplitude Min/Max"=<0 1>
     "[1] Exponent"=3
    )
    Use Spawn Probability=false
   )
  ]
  particleSpawnScript [
   &InitializeParticle(
    Lifetime=&MaskFloatBySpawnGroup(Masked Float=&RandomRangeFloat(Minimum=0.4, Maximum=0.5), Passthrough Float=0.5)
    Mesh Scale Mode=::NewEnumerator3
    Mesh Uniform Scale=&RandomRangeFloat(Minimum=0.2, Maximum=0.4)
    Sprite Size Mode=::NewEnumerator3
    Uniform Sprite Size=20
    UsePositionOffset=false
    Write Material Random=true
   )
   /SetVariables(Particles.Position=&Multiply_VectorByFloat(
     Vector=/Niagara/DynamicInputs/Arrays/SelectPositionFromArray(
      Array Sampling Mode:ENiagaraArraySamplingMode=::NewEnumerator0
      Direct Array Index:NiagaraInt32=?
      Interpolation:NiagaraFloat=?
      Direct Array Mode:ENiagara_WrapClamp=::NewEnumerator0
      Evaluation Type:ENiagaraRandomnessEvaluation=::NewEnumerator0
      Normalized Interpolation:NiagaraBool=true
      Recalculate Random Each Loop:NiagaraBool=false
      Position Array:NiagaraDataInterfaceArrayPosition=DataInterface() {propertyValues {
        GpuSyncMode=SyncCpuToGpu
        MaxElements=0
        PositionData [
         <192.3417 -106.3385 56.2992>
         <92.6447 -122.9211 43.791>
         <-13.0827 109.5889 17.7729>
         <142.0064 0.1204 93.4829>
         <29.4106 -113.9767 25.7462>
         <116.2699 111.1617 26.0482>
         <117.2309 -109.3109 33.1014>
         <-177.3799 83.9769 40.9377>
         <249.6831 -21.8321 21.3158>
         <3.9345 81.9685 97.2264>
         <168.9551 54.324 86.5875>
         <222.9013 -19.8242 72.0191>
         <246.6383 -48.9313 22.7249>
         <-95.7748 -93.6136 132.8081>
         <-193.8617 15.2701 116.4681>
         <241.4438 29.5988 32.5795>
         <131.6617 -103.186 86.5738>
         <166.6924 -42.0642 87.723>
         <221.2293 32.9973 72.3963>
         <104.2041 -104.7772 15.3916>
         <37.5161 -119.9627 65.6404>
         <-163.4495 -100.2206 114.2653>
         <206.3187 -31.4263 77.7426>
         <-35.2132 -56.8897 138.6964>
         <-102.0604 -116.6031 86.6792>
         <9.4449 -94.3526 18.9676>
         <-61.4735 122.3095 81.8644>
         <172.3155 33.537 89.8953>
         <-2.5427 -125.9877 51.4818>
         <118.8014 -66.6175 95.3649>
         <255.069 -13.6363 57.9673>
         <-160.624 -81.3733 122.5638>
         <210.9325 -4.3932 82.2372>
         <153.522 -103.3489 83.1329>
         <-222.3238 21.796 107.5272>
         <184.1791 -95.5657 79.2045>
         <-215.3556 76.6847 73.8825>
         <237.3978 11.1751 66.4705>
         <-181.7926 -61.7395 113.1287>
         <150.9861 22.0087 93.7708>
         <-140.5879 -108.2859 110.999>
         <-187.8447 -109.8502 82.3367>
         <-31.5395 -116.0174 27.5626>
         <-82.8378 -118.3023 101.884>
         <119.7693 84.2747 93.5966>
         <-160.3679 -116.5473 76.8586>
         <246.1191 69.8206 36.1936>
         <166.8189 8.5702 88.7516>
         <60.768 122.0367 57.4476>
         <-155.4475 52.9177 123.1371>
         <131.294 16.9034 96.3179>
         <261.5865 19.8383 38.1771>
         <-62.2791 -120.3975 87.0202>
         <-52.4525 -95.134 19.9279>
         <-9.6441 -120.3448 69.4248>
         <160.6771 104.4142 80.701>
         <-213.9655 -37.8039 72.0559>
         <57.494 -119.2154 66.2876>
         <269.1017 54.5713 43.8613>
         <59.2696 -101.5779 93.1854>
         <-42.2465 117.2781 84.1296>
         <190.8019 -7.8766 82.8127>
         <-201.9474 55.1033 110.944>
         <-12.6958 -108.1583 92.674>
         <170.2475 -109.5751 70.5293>
         <112.6074 65.8097 96.5755>
         <-216.1889 56.5053 73.0691>
         <204.5027 38.7701 83.298>
         <-139.8525 -84.3482 126.6283>
         <71.0756 -112.9317 79.6364>
         <141.8727 -110.2615 68.1724>
         <180.8964 -27.7131 88.2717>
         <-126.5531 -99.3776 123.4007>
         <5.9633 103.9868 14.1969>
         <-63.6801 -128.6146 65.5422>
         <68.7898 -113.0872 25.1632>
         <-46.3381 -108.4863 100.5344>
         <245.4742 93.4478 34.7195>
         <16.3784 -124.8169 44.9802>
         <-81.0464 -125.9165 80.1332>
         <254.1365 -36.0837 37.2053>
         <-158.2162 -17.1685 123.7894>
         <-8.4272 37.9598 138.425>
         <119.9362 -113.8702 71.5067>
         <62.3185 88.6571 96.4136>
         <-145.2928 102.6469 117.2994>
         <-104.5249 11.8364 137.4016>
         <-152.76 110.8747 96.6106>
         <46.951 116.9957 30.8236>
         <-117.1088 91.4811 129.4365>
         <123.4621 118.8279 55.5483>
         <-190.2108 -37.2764 118.0378>
         <76.1045 103.8814 91.5534>
         <55.6949 107.4453 88.0284>
         <-153.2521 85.0017 123.6615>
         <-222.9449 -72.0291 92.8439>
         <-181.7458 52.9996 116.9386>
         <-187.341 -3.9136 47.4172>
         <-149.4652 9.8206 20.9387>
         <-128.4521 56.3787 129.2046>
        ]
       }}
      Randomness Mode:ENiagaraRandomnessMode=::NewEnumerator0
      Fixed Random Seed:NiagaraBool=false
      Random Seed:NiagaraInt32=0
     )
     Float=1
    )) {isSetParameters=true}
   &InitialMeshOrientation(
    Mesh Orientation Mode=::NewEnumerator2
    Orientation Vector=&RandomVector(Vector Scale=1)
    Use Rotation Vector=true
   )
   &CameraOffset(Camera Offset Amount=10)
  ]
  particleUpdateScript [
   &ParticleState()
   &ScaleMeshSize(Scale Factor=&Multiply_VectorByFloat(
     Vector=&RandomRangeVector(Minimum=<1 1 2.2>, Maximum=<1.3 1.3 2.2>)
     Float=&FloatFromCurve(FloatCurve=DataInterface() {propertyValues {Curve {keys [<at 0 1>,<at 1 1.5>]}}})
    )) {module=ScaleMeshSize001}
   &DynamicMaterialParameters(
    Index 0 Param 2=&RandomRangeFloat(Minimum=0, Maximum=0)
    Index 0 Param 4=&MaskFloatBySpawnGroup(Masked Float=13, Passthrough Float=1)
    Index 1 Param 1=&FloatFromCurve(FloatCurve=DataInterface() {propertyValues {Curve {keys [<at 0 0.5>,<at 1 1>]}}}, Scale Curve=0.5)
    Index 1 Param 2=2
    Index 1 Param 3=&FloatFromCurve(FloatCurve=DataInterface() {propertyValues {Curve {keys [{interpMode=RCIM_Cubic, tangentMode=RCTM_Break, time=0.5, value=-0.5}, <at 1 0.1>]}}})
    Index 2 Param 1=&Multiply_Float(
     A=&FloatFromCurve(
      FloatCurve=DataInterface() {propertyValues {Curve {keys [{interpMode=RCIM_Cubic, leaveTangent=8.1085, tangentMode=RCTM_Break, time=0.8568, value=1}, {tangentMode=RCTM_User, time=1, value=0}]}}}
      Scale Curve=&MaskFloatBySpawnGroup(Masked Float=1, Passthrough Float=1)
     )
     B=&Divide_Float(
      A=&MakeFloatFromVector(Vector=$Particles.Scale, Channel=::NewEnumerator0)
      B=&RandomRangeFloat(Minimum=10, Maximum=15)
     )
    )
    Index 2 Param 2=&FloatFromCurve(
     FloatCurve=DataInterface() {propertyValues {Curve {keys [
         {interpMode=RCIM_Cubic, tangentMode=RCTM_Break, time=0.7, value=0}
         {arriveTangent=-0.0859, interpMode=RCIM_Cubic, tangentMode=RCTM_Break, time=1, value=-1}
        ]}}}
     Scale Curve=50
    )
    Index 2 Param 3=&Waveform(
     Global Input Position=&RandomRangeFloat(Minimum=0, Maximum=15)
     Global Phase=$Particles.Age
     "Waveform [1]"=::NewEnumerator12
     "[1] Phase Scale"=2
     "[1] Frequency"=1
     "[1] Amplitude Min/Max"=<0 1>
    )
    Index 2 Param 4=25
    Index 3 Param 1=&FloatFromCurve(
     FloatCurve=DataInterface() {propertyValues {Curve {keys [
         {interpMode=RCIM_Cubic, leaveTangent=-0.0025, tangentMode=RCTM_Break, time=0, value=0}
         {arriveTangent=1.8528, interpMode=RCIM_Cubic, tangentMode=RCTM_Break, time=1, value=1}
        ]}}}
     Scale Curve=&RandomRangeFloat(Minimum=0, Maximum=0)
    )
    Index 3 Param 2=&FloatFromCurve(
     FloatCurve=DataInterface() {propertyValues {Curve {keys [
         {interpMode=RCIM_Cubic, leaveTangent=-0.0025, tangentMode=RCTM_Break, time=0, value=0.5}
         {arriveTangent=-0.0316, interpMode=RCIM_Cubic, tangentMode=RCTM_Break, time=1, value=-0.2}
        ]}}}
     Scale Curve=0.1
    )
    Index 3 Param 3=&FloatFromCurve(FloatCurve=DataInterface() {propertyValues {Curve {keys [
         {interpMode=RCIM_Cubic, leaveTangent=1.1034, tangentMode=RCTM_Break, time=0, value=-1}
         {arriveTangent=2.2399, tangentMode=RCTM_User, time=0.7, value=0.125}
        ]}}})
    Index 3 Param 4=&FloatFromCurve(FloatCurve=DataInterface() {propertyValues {Curve {keys [
         {interpMode=RCIM_Cubic, leaveTangent=3.0601, tangentMode=RCTM_Break, time=0.5, value=0}
         {arriveTangent=3.2053, tangentMode=RCTM_User, time=1, value=1}
        ]}}})
    Module.Index0Param0Write=true
    Module.Index0Param1Write=true
    Module.Index0Param2Write=true
    Module.Index0Param3Write=true
    Module.Index1Param0Write=true
    Module.Index1Param1Write=true
    Module.Index1Param2Write=true
    Module.Index1Param3Write=true
    Module.Index2Param0Write=true
    Module.Index2Param1Write=true
    Module.Index2Param2Write=true
    Module.Index2Param3Write=true
    Module.Index3Param0Write=true
    Module.Index3Param1Write=true
    Module.Index3Param2Write=true
    Module.Index3Param3Write=true
    Write Parameter Index 1=true
    Write Parameter Index 2=true
    Write Parameter Index 3=true
   )
   &SolveForcesAndVelocity(Clamp Velocity=false, Limit Acceleration=false)
   &Color(
    Color=&ColorFromCurve(ColorCurve=DataInterface() {propertyValues {
       AlphaCurve {keys [<at 0 1 RCIM_Cubic>,<at 0.7884 1>,<at 1 0 RCIM_Cubic>]}
       BlueCurve {keys [<at 0.2225 5>, {interpMode=RCIM_Cubic, leaveTangent=-19.7604, time=0.4906, value=3}]}
       GreenCurve {keys [<at 0.2225 2.4153>, {interpMode=RCIM_Cubic, leaveTangent=-19.7604, time=0.4906, value=0.4144}]}
       RedCurve {keys [<at 0.2225 1.7969>, {interpMode=RCIM_Cubic, leaveTangent=-19.7604, time=0.4906, value=0.2513}]}
      }})
    Scale Alpha=&FloatFromCurve(FloatCurve=DataInterface() {propertyValues {Curve {keys [
         {interpMode=RCIM_Cubic, leaveTangent=2.5294, tangentMode=RCTM_Break, time=0, value=0}
         {arriveTangent=-0.2166, interpMode=RCIM_Cubic, tangentMode=RCTM_Break, time=0.4, value=1}
        ]}}})
   )
   /SetVariables(Particles.DecalColor:LinearColor=&MultiplyLinearColor(Linear Color=<rgba 1 1 1 1>, Scale Factor=&MakeVector4_FromLinearColor(LinearColor=$Particles.Color))) {isSetParameters=true}
   /SetVariables(Particles.GlowDecalSize:Vector2f=&Multiply_Vector2D_ByFloat(Float=80, Vector2D=<1 1>)) {isSetParameters=true}
   /SetVariables(Particles.DecalDynMat:Vector4f=<7 1 0 0>) {isSetParameters=true}
  ]
  rendererClasses [&NiagaraMeshRendererProperties, &NiagaraSpriteRendererProperties]
  renderers [
   {rendererClass=&NiagaraMeshRendererProperties, rendererIndex=0}
   {rendererClass=&NiagaraSpriteRendererProperties, rendererIndex=1}
   {rendererClass=&NiagaraMeshRendererProperties, rendererIndex=2}
  ]
 }
 Portal_LW {rendererClasses [], renderers []}
 SpeedLines {
  emitterUpdateScript [
   &EmitterState()
   &SpawnRate(
    SpawnRate=&FloatFromCurve(
     FloatCurve=DataInterface() {propertyValues {Curve {defaultValue=1, keys [<at 0 0 RCIM_Cubic>, {tangentMode=RCTM_User, time=1, value=1}]}}}
     CurveIndex=$System.Timeline.NormalizedTime
     Scale Curve=10
    )
    Use Spawn Probability=false
   )
   /SetVariables(Emitter.DynamicInertia:NiagaraFloat=&Lerp_Float(A=1, B=0.3, Alpha=$System.Speed_Alpha)) {isSetParameters=true}
  ]
  particleSpawnScript [
   &InitializeParticle(
    Color=<rgba 1.2083 1.8835 4 1>
    Lifetime Max=0.6
    Lifetime Min=0.4
    Lifetime Mode=::NewEnumerator1
    Mesh Renderer Array Visibility Mode=::NewEnumerator1
    Mesh Renderer Info=DataInterface() {propertyValues {MeshRenderer=&self:SpeedLines_0.NiagaraMeshRendererProperties_0}}
    Mesh Renderer Mesh Index:NiagaraInt32=&MakeIntFromFloat(
     Float=&FloatFromCurve(
      FloatCurve=DataInterface() {propertyValues {Curve {defaultValue=1, keys [<at 0 0 RCIM_Cubic>, {arriveTangent=2.8878, tangentMode=RCTM_User, time=1, value=1}]}}}
      CurveIndex=&RandomRangeFloat(Minimum=0, Maximum=1)
     )
     Conversion Mode=::NewEnumerator1
    )
    Mesh Scale=&Lerp_Vector(
     A=<0.8 0.8 0.8>
     B=<1 1 1>
     Alpha=&FloatFromCurve(
      FloatCurve=DataInterface() {propertyValues {Curve {
         defaultValue=1
         keys [
          {arriveTangent=0.4663, interpMode=RCIM_Cubic, leaveTangent=0.4663, tangentMode=RCTM_User, time=0, value=0}
          {arriveTangent=2.2578, tangentMode=RCTM_User, time=1, value=1}
         ]
        }}}
      CurveIndex=&RandomRangeFloat(Minimum=0, Maximum=1)
     )
    )
    Mesh Scale Mode=::NewEnumerator4
    Sprite Rotation Mode=::NewEnumerator2
    Sprite Size Mode=::NewEnumerator3
    Uniform Sprite Size=&FloatFromCurve(
     FloatCurve=DataInterface() {propertyValues {Curve {
        defaultValue=1
        keys [
         {arriveTangent=41.0327, interpMode=RCIM_Cubic, leaveTangent=41.0327, tangentMode=RCTM_User, time=0, value=50}
         {arriveTangent=820.3138, interpMode=RCIM_Cubic, leaveTangent=820.3138, tangentMode=RCTM_User, time=1, value=400}
        ]
       }}}
     CurveIndex=&RandomRangeFloat(Minimum=0, Maximum=1)
    )
    UsePositionOffset=false
    Write Material Random=true
   )
   &SkeletalMeshLocation(
    MaskSpawnGroup=false
    Mesh Sampling Type=::NewEnumerator5
    Orientation Sampling=::NewEnumerator1
    Sampled Position Offset=<60 0 20>
    Skeletal Mesh=$System.SkeletalMesh
    Socket Sampling Mode=::NewEnumerator5
   )
   &ShapeLocation(Box Size=<5 180 30>, MaskSpawnGroup=false, Override Seed=false, Shape Primitive=::NewEnumerator2, Surface Only=false, UseOwnerScale=true)
   &InitialMeshOrientation(
    Mesh Orientation Mode=::NewEnumerator2
    Rotation=&Lerp_Vector(
     A=<0 0 0>
     B=&RandomRangeVector(Minimum=<0.2 0 0>, Maximum=<-1.4 0 0>)
     Alpha=&MakeFloatFromVector(
      Vector=&Divide_Vector(
       A:Vector3f=&ConvertPositionToVector(Input Position=$Particles.Position, Conversion Operation:ENiagara_LWCConvertPosToVec=::NewEnumerator0)
       B:Vector3f=<1 200 1>
      )
      Channel=::NewEnumerator1
     )
    )
    Use Rotation Vector=true
   )
  ]
  particleUpdateScript [
   &ParticleState()
   &ScaleColor(
    Scale Alpha=&FloatFromCurve(
     FloatCurve=DataInterface() {propertyValues {Curve {
        defaultValue=1
        keys [
         {arriveTangent=9.0725, interpMode=RCIM_Cubic, leaveTangent=9.0725, tangentMode=RCTM_User, time=0, value=0}
         <at 0.1408 0.9847 RCIM_Cubic>
         {arriveTangent=-0.0405, interpMode=RCIM_Cubic, leaveTangent=-0.0405, tangentMode=RCTM_User, time=1, value=0}
        ]
       }}}
     Scale Curve=0.8
    )
    Scale RGB=&VectorFromFloat(Value=&FloatFromCurve(
      FloatCurve=DataInterface() {propertyValues {Curve {
         defaultValue=1
         keys [
          {arriveTangent=9.0725, interpMode=RCIM_Cubic, leaveTangent=9.0725, tangentMode=RCTM_User, time=0, value=0}
          <at 0.0916 0.9847 RCIM_Cubic>
          {arriveTangent=-0.0405, interpMode=RCIM_Cubic, leaveTangent=-0.0405, tangentMode=RCTM_User, time=1, value=0}
         ]
        }}}
      Scale Curve=2
     ))
    ScaleA=true
    ScaleRGB=true
   )
   &ScaleSpriteSize(Uniform Curve Sprite Scale=DataInterface() {propertyValues {Curve {
       defaultValue=0
       keys [
        {arriveTangent=4.1785, interpMode=RCIM_Cubic, leaveTangent=4.1785, tangentMode=RCTM_User, time=0, value=0}
        <at 0.1 1 RCIM_Cubic>
        <at 0.5 0 RCIM_Cubic>
       ]
      }}})
   &DynamicMaterialParameters(
    Index 0 Param 1=&FloatFromCurve(FloatCurve=DataInterface() {propertyValues {Curve {
        defaultValue=1
        keys [
         {arriveTangent=0.5948, interpMode=RCIM_Cubic, leaveTangent=0.5948, tangentMode=RCTM_User, time=0, value=0}
         {arriveTangent=1.7748, interpMode=RCIM_Cubic, leaveTangent=1.7748, tangentMode=RCTM_User, time=0.1486, value=0.2622}
         {arriveTangent=2.3975, tangentMode=RCTM_User, time=1, value=2}
        ]
       }}})
    Index 0 Param 2=1
    Index 0 Param 3=0.1
    Index 0 Param 4=1
    Index 1 Param 2=1
    Index 1 Param 3=&FloatFromCurve(
     FloatCurve=DataInterface() {propertyValues {Curve {
        defaultValue=1
        keys [
         {arriveTangent=9.0725, interpMode=RCIM_Cubic, time=0, value=0}
         {arriveTangent=0.1521, interpMode=RCIM_Cubic, leaveTangent=0.1521, time=0.6197, value=0.9616}
         {arriveTangent=0.0745, interpMode=RCIM_Cubic, leaveTangent=0.0745, tangentMode=RCTM_User, time=1, value=0.1521}
        ]
       }}}
     Scale Curve=&RandomRangeFloat(Minimum=2, Maximum=7)
    )
    Index 1 Param 4=&FloatFromCurve(
     FloatCurve=DataInterface() {propertyValues {Curve {
        defaultValue=1
        keys [
         {arriveTangent=9.0725, interpMode=RCIM_Cubic, leaveTangent=9.0725, tangentMode=RCTM_User, time=0, value=0}
         {arriveTangent=-2.5249, interpMode=RCIM_Cubic, leaveTangent=-2.5249, tangentMode=RCTM_User, time=0.0369, value=1}
         {arriveTangent=-0.0348, interpMode=RCIM_Cubic, leaveTangent=-0.0348, tangentMode=RCTM_User, time=0.1131, value=0.3234}
         {arriveTangent=-0.0405, interpMode=RCIM_Cubic, leaveTangent=-0.0405, tangentMode=RCTM_User, time=1, value=0.1}
        ]
       }}}
     Scale Curve=&RandomRangeFloat(Minimum=10, Maximum=20)
    )
    Index 2 Param 1=&FloatFromCurve(
     FloatCurve=DataInterface() {propertyValues {Curve {
        defaultValue=1
        keys [
         {arriveTangent=2.6165, interpMode=RCIM_Cubic, leaveTangent=2.6165, tangentMode=RCTM_User, time=0, value=0}
         {tangentMode=RCTM_User, time=1, value=1}
        ]
       }}}
     Scale Curve=0.1
    )
    Index 2 Param 2=&RandomRangeFloat(Minimum=0.2, Maximum=1)
    Index 2 Param 3=&FloatFromCurve(
     FloatCurve=DataInterface() {propertyValues {Curve {
        defaultValue=0
        keys [
         {arriveTangent=4.1785, interpMode=RCIM_Cubic, leaveTangent=4.1785, tangentMode=RCTM_User, time=0, value=0}
         {arriveTangent=1.1179, interpMode=RCIM_Cubic, leaveTangent=1.1179, tangentMode=RCTM_User, time=0.2042, value=0.6228}
         {arriveTangent=0.0224, interpMode=RCIM_Cubic, leaveTangent=0.0224, tangentMode=RCTM_User, time=1, value=1}
        ]
       }}}
     Scale Curve=&RandomRangeFloat(Minimum=1, Maximum=3)
    )
    Index 2 Param 4=&FloatFromCurve(FloatCurve=DataInterface() {propertyValues {Curve {
        defaultValue=1
        keys [
         <at 0.0002 2 RCIM_Cubic>
         {arriveTangent=-1.0002, interpMode=RCIM_Cubic, leaveTangent=-1.0002, time=0.1495, value=0}
         {arriveTangent=3.4029, tangentMode=RCTM_User, time=1, value=1}
        ]
       }}})
    Module.Index0Float0Write=true
    Module.Index0Float1Write=true
    Module.Index0Float2Write=true
    Module.Index0Float3Write=true
    Module.Index1Float0Write=true
    Module.Index1Float1Write=true
    Module.Index1Float2Write=true
    Module.Index1Float3Write=true
    Module.Index2Float0Write=true
    Module.Index2Float1Write=true
    Module.Index2Float2Write=true
    Module.Index2Float3Write=true
    Module.Index3Float0Write=true
    Module.Index3Float1Write=true
    Module.Index3Float2Write=true
    Module.Index3Float3Write=true
    Write Parameter Index 1=::NewEnumerator1
    Write Parameter Index 2=::NewEnumerator1
    Write Parameter Index 3=::NewEnumerator1
   )
   /SetVariables(
    Particles.Color - Bloom:LinearColor=&ColorFromCurve(ColorCurve=DataInterface() {propertyValues {
       AlphaCurve {keys [<at 0.0056 0>,<at 0.3754 0.25 RCIM_Cubic>,<at 1 0 RCIM_Cubic>]}
       BlueCurve {keys [<at 0 1 RCIM_Cubic>,<at 0.9923 0.5677>]}
       GreenCurve {keys [<at 0 0.355 RCIM_Cubic>,<at 0.9923 0.0565>]}
       RedCurve {keys [<at 0 0.0498 RCIM_Cubic>,<at 0.9923 0.0198>]}
      }})
    "Particles.DynamicMaterialParameter-Bloom ":Vector4f=&MakeVector4(
     X=0
     Y=0
     Z=&FloatFromCurve(
      FloatCurve=DataInterface() {propertyValues {Curve {defaultValue=1, keys [<at 0 0 RCIM_Cubic>, {arriveTangent=1.6014, tangentMode=RCTM_User, time=1, value=1}]}}}
      Scale Curve=120
     )
     W=&FloatFromCurve(
      FloatCurve=DataInterface() {propertyValues {Curve {keys [{arriveTangent=-0.0966, interpMode=RCIM_Cubic, leaveTangent=-0.0966, tangentMode=RCTM_User, time=0, value=1}, <at 1 0>]}}}
      Scale Curve=140
     )
    )
   ) {isSetParameters=true}
   &self:SampleCurveArray(
    Alpha:NiagaraFloat=&FloatFromCurve(FloatCurve=DataInterface() {propertyValues {Curve {
        defaultValue=1
        keys [
         {arriveTangent=1.4739, interpMode=RCIM_Cubic, leaveTangent=1.4739, tangentMode=RCTM_User, time=0, value=0}
         {arriveTangent=1.6818, interpMode=RCIM_Cubic, leaveTangent=1.6818, tangentMode=RCTM_User, time=0.2, value=0.3538}
         {arriveTangent=0.4519, tangentMode=RCTM_User, time=1, value=1}
        ]
       }}})
    CurveSampler:NiagaraInt32=$Particles.MeshIndex
    Orientation:Quat4f=$Particles.Initial.MeshOrientation
    PointPerCurve=16
    PositionOffset:NiagaraPosition=$Particles.Initial.Position
    ScaleCurve:Vector3f=$Particles.Scale
    VectorArray=DataInterface() {propertyValues {
      GpuSyncMode=SyncCpuToGpu
      MaxElements=0
      FloatData [
       <0.2899 -0.2705 22.8579>
       <-28.6608 0.0099 55.2287>
       <-65.2707 0.1111 79.6985>
       <-104.3202 0.1343 100.2691>
       <-144.9154 0.1343 117.12>
       <-187.1029 0.1342 129.7308>
       <-230.2289 0.1342 138.2894>
       <-273.8496 0.1342 143.2696>
       <-317.7838 0.1341 145.2289>
       <-361.7922 0.1341 144.6831>
       <-405.7221 0.1341 142.0941>
       <-449.5009 0.1341 137.8497>
       <-493.1212 0.1341 132.2606>
       <-536.6249 0.1341 125.5684>
       <-580.0875 0.1341 117.9561>
       <-623.6061 0.1341 109.5599>
      ]
     }}
   ) {module=SampleCurveArray}
   /SetVariables(
    Particles.Color - Flare:LinearColor=&MultiplyLinearColor(
     Linear Color=$Particles.Initial.Color
     Scale Factor=&MakeVector4FromVectorAndFloat(
      "Vector (RGB)"=&VectorFromFloat(Value=&FloatFromCurve(
        FloatCurve=DataInterface() {propertyValues {Curve {
           defaultValue=1
           keys [
            {arriveTangent=0.556, interpMode=RCIM_Cubic, leaveTangent=0.556, tangentMode=RCTM_User, time=0, value=0}
            {arriveTangent=-0.2562, interpMode=RCIM_Cubic, leaveTangent=-0.2562, tangentMode=RCTM_User, time=0.1016, value=0.8685}
            {arriveTangent=-0.0405, interpMode=RCIM_Cubic, leaveTangent=-0.0405, tangentMode=RCTM_User, time=0.4, value=0}
           ]
          }}}
        Scale Curve=&RandomRangeFloat(Minimum=0.8, Maximum=1.2)
       ))
      "Float (Alpha)"=2
     )
    )
    Particles.Position - Flare:Vector3f=$Output.SampleCurveArray.CurveSampledVector
    Particles.SpriteSize - Flare:Vector2f=&Vector2DFromFloat(Value=&RandomRangeFloat(Minimum=80, Maximum=220))
   ) {isSetParameters=true}
   /SetVariables(
    Particles.DecalColor:LinearColor=&MultiplyLinearColor(Linear Color=Linked() {linkedVariable {name=Particles.Color - Flare, type=LinearColor}}, Scale Factor=<1 1 1 1>)
    Particles.DecalDynMat:Vector4f=<7 1 0 0>
    Particles.GlowDecalSize:Vector2f=&Vector2DFromFloat(Value=&FloatFromCurve(
      FloatCurve=DataInterface() {propertyValues {Curve {
         defaultValue=1
         keys [
          {arriveTangent=0.5071, interpMode=RCIM_Cubic, leaveTangent=0.5071, tangentMode=RCTM_User, time=0, value=1}
          {arriveTangent=-0.3075, interpMode=RCIM_Cubic, leaveTangent=-0.3075, tangentMode=RCTM_User, time=0.264, value=0.9806}
          {arriveTangent=0.5071, interpMode=RCIM_Cubic, leaveTangent=0.5071, tangentMode=RCTM_User, time=1, value=0}
         ]
        }}}
      Scale Curve=150
     ))
   ) {isSetParameters=true}
  ]
  rendererClasses [&NiagaraMeshRendererProperties, &NiagaraSpriteRendererProperties]
  renderers [
   {rendererClass=&NiagaraMeshRendererProperties, rendererIndex=0}
   {rendererClass=&NiagaraMeshRendererProperties, rendererIndex=1}
   {rendererClass=&NiagaraSpriteRendererProperties, rendererIndex=2}
   {rendererClass=&NiagaraSpriteRendererProperties, rendererIndex=3}
   {rendererClass=&NiagaraSpriteRendererProperties, rendererIndex=4}
  ]
 }
}
summary {
 systemName=NS_Ambient_GreyCar
 emitters [
  {emitterName=SpeedLines, rendererClasses [&NiagaraMeshRendererProperties, &NiagaraSpriteRendererProperties]}
  {emitterName=Flames, rendererClasses [&NiagaraMeshRendererProperties, &NiagaraSpriteRendererProperties]}
  {bEnabled=false, emitterName=BoostFlare_LW, rendererClasses []}
  {emitterName=Booster001, simTarget=GPUComputeSim, rendererClasses [&NiagaraSpriteRendererProperties]}
  {emitterName=NE_Zaps, rendererClasses [&NiagaraMeshRendererProperties, &NiagaraSpriteRendererProperties]}
  {emitterName=Portal_LW, rendererClasses []}
 ]
 userVariables [
  User.bIsBoosting {type=NiagaraBool}
  User.bIsFlying {type=NiagaraBool}
  User.bIsFrontEnd {type=NiagaraBool}
  User.bIsFrontendPreview {type=NiagaraBool}
 ]
}
