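# Material function for the GreyCar Hypercolor V2 skin, an animated paint that reacts to a Speedometer parameter.
# Panned, mutually distorting samples of an RGB texture blend team, paint and custom colors through hue shifts and Fresnel, and also build the normal.
# Speed (or a debug sine) plus noise and a UV gradient forms soft and sharp masks for base color and edge glow.
# Results feed SetMaterialAttributes alongside the vehicle body base function.
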MaterialFunction /VehicleCosmetics/Bodies/GreyCar/Skins/MF_Skin_HypercolorV2
graph {nodes [
  n1 = NamedRerouteUsage()
  n2 = NamedRerouteUsage()
  n3 = Constant3Vector()
  n4 = NamedRerouteUsage()
  n5 = Multiply(A=@n4, B=ScalarParameter() {defaultValue=1, parameterName=EmissiveIntensity})
  n6 = SetMaterialAttributes!(
   MaterialAttributes
   BaseColor=@n1
   Metallic=ScalarParameter() {defaultValue=1, parameterName=Metallic}
   Roughness=ScalarParameter() {defaultValue=0.5, parameterName=Roughness}
   CustomData0
   CustomData1
   Normal=NamedRerouteUsage()
   ClearCoatBottomNormal
   EmissiveColor=Add(A=@n2, B=LinearInterpolate(A=@n3, B=@n1, Alpha=@n5))
   ShadingModel
   OpacityMask=Max(A=@n4, B=Desaturation(None=@n2))
  ) {attributeSetTypes [
    #1
    #2
    #3
    #4
    #5
    #6
    #7
    #8
    #9
    #10
   ]}
  FunctionOutput(None=@n6)
  GetMaterialAttributes(None=MaterialFunctionCall() {materialFunction=/VehicleCosmetics/Materials/VehicleParents_Version2/Material_Functions/MF_Vehicle_Body_Base_v2}.Result) {attributeGetTypes [
    #2
    #3
    #8
    #4
    #5
    #7
    #9
   ]}
  NamedRerouteDeclaration(None=Time() {bOverride_Period=true, period=600}) {name=Time, nodeColor=<rgba 0.5469 1 0 1>}
  n7 = TextureCoordinate() {coordinateIndex=1}
  n8 = NamedRerouteUsage()
  n9 = Rotator(Coordinate=@n7, Time=@n8) {speed=0.02}
  n10 = TextureSampleParameter2D(UVs=Panner(Coordinate=@n7) {bFractionalPart=true, speedX=0.2, speedY=0.2}) {parameterName=RGB, texture=/VehicleCosmetics/Bodies/MuscleCar/Skins/Hypercolor_RGB}
  n11 = TextureCoordinate() {coordinateIndex=1, uTiling=0.5, vTiling=0.2}
  n12 = TextureSampleParameter2D(UVs=Add(A=Multiply(A=@n10.B) {constB=0.015}, B=Panner(Coordinate=@n11) {bFractionalPart=true, speedX=0.002, speedY=0.15})) {parameterName=RGB, texture=&Hypercolor_RGB}
  n13 = Multiply(B=@n12.G) {constA=0.5, constB=0.5}
  n14 = TextureSampleParameter2D(UVs=Add(A=Panner(Coordinate=@n9, Time=@n8) {bFractionalPart=true, speedX=0.1, speedY=0.125}, B=@n13)) {parameterName=RGB, texture=&Hypercolor_RGB}
  n15 = AppendVector(A=AppendVector(A=@n14.B, B=@n14.B), B=Constant() {r=1})
  NamedRerouteDeclaration(None=LinearInterpolate(A=Constant3Vector() {constant=<rgba 0 0 1 0>}, B=@n15, Alpha=NamedRerouteUsage())) {name=Normal, nodeColor=<rgba 0.0781 0 1 1>}
  n16 = Multiply(A=@n12.G) {constB=0.4}
  n17 = ComponentMask(None=@n7) {g=true}
  n18 = Sine(None=Time()) {period=5}
  n19 = ScalarParameter() {desc="Speedometer - uses percentage\r\nof speed from VCID", parameterName=Speedometer}
  n20 = Saturate(None=ConstantBiasScale(None=StaticSwitchParameter(True=@n18, False=@n19) {parameterName=SpeedDebug}) {bias=0, desc=Balance copied from Pixie, scale=1})
  n21 = Add(A=@n20) {constB=0.5}
  n22 = Add(A=Multiply(A=Add(A=@n16, B=@n17), B=Saturate(None=@n21)), B=@n20)
  n23 = Multiply(A=@n22) {constB=0.98}
  NamedRerouteDeclaration(None=Saturate(None=Floor(None=@n23))) {name=blendSharp, nodeColor=<rgba 1 0 0.5625 1>}
  NamedRerouteDeclaration(None=Saturate(None=ConstantBiasScale(None=@n22) {bias=-1, scale=5})) {name=blend, nodeColor=<rgba 0.2422 1 0 1>}
  NamedRerouteDeclaration(None=ScalarParameter() {defaultValue=0.06, parameterName=HueShiftMultiplier}) {name=HueShiftMult, nodeColor=<rgba 0.7578 1 0 1>}
  n24 = VectorParameter() {parameterName=HyperCustomColor, defaultValue=<rgba 1 0.4511 0.0878 1>}
  n25 = Multiply(A=@n24.RGB) {constB=30}
  n26 = NamedRerouteUsage()
  n27 = NamedRerouteUsage()
  NamedRerouteDeclaration(None=LinearInterpolate(A=LinearInterpolate(B=@n25, Alpha=@n26), Alpha=@n27) {constB=0}) {name=edgeglow, nodeColor=<rgba 1 0 0.0469 1>}
  n28 = VectorParameter() {parameterName=HyperTeamColor, defaultValue=<rgba 0.1358 0.3097 1 1>}
  n29 = TextureCoordinate() {coordinateIndex=1, vTiling=0.25}
  n30 = ComponentMask(None=Add(A=Panner(Coordinate=@n29) {bFractionalPart=true, speedY=0.1}, B=@n13)) {g=true}
  n31 = Subtract(A=@n30, B=Floor(None=@n30))
  n32 = ConstantBiasScale(None=@n31) {bias=-0.5, scale=2}
  n33 = Clamp(None=ConstantBiasScale(None=Abs(None=@n32)) {bias=-0.5, scale=0.1}) {minDefault=-0.2}
  n34 = VectorParameter() {parameterName=HyperPaintColor, defaultValue=<rgba 1 0 0.1826 1>}
  n35 = MaterialFunctionCall(Hue Shift Percentage=@n33, Texture=LinearInterpolate(A=LinearInterpolate(A=@n24.RGB, B=@n34.RGB, Alpha=@n14.B), B=@n24.RGB, Alpha=@n10.B)) {materialFunction=&HueShift}
  n36 = MaterialFunctionCall!(Fresnel Power=Constant() {r=0.7}, Normal=@n15) {materialFunction=/VehicleCosmetics/Materials/MaterialFunctions/MF_Fresnel_Prime}
  n37 = ComponentMask(None=Panner(Coordinate=@n36.Result, Time=NamedRerouteUsage()) {bFractionalPart=true, speedY=-0.2}) {g=true}
  n38 = Clamp(None=Subtract(A=@n37, B=Floor(None=@n37))) {maxDefault=2}
  n39 = ConstantBiasScale(None=@n33) {bias=-0.2, scale=-0.02}
  n40 = LinearInterpolate(
   A=LinearInterpolate(A=LinearInterpolate(A=@n28.RGB, B=@n35.Result, Alpha=@n12.G), B=@n24.RGB, Alpha=Multiply(A=@n38, B=@n31) {constB=0.06})
   B=MaterialFunctionCall(Hue Shift Percentage=@n39, Texture=@n24.RGB) {materialFunction=&HueShift}.Result
   Alpha=@n12.G
  )
  n41 = MaterialFunctionCall(Hue Shift Percentage=@n39, Texture=@n28.RGB) {materialFunction=&HueShift}
  n42 = Saturate(None=ConstantBiasScale(None=@n36.Result) {bias=-0.1, scale=1.5})
  n43 = Saturate(None=ConstantBiasScale(None=@n36.Result) {bias=-0.2, scale=0.3})
  n44 = Multiply(A=@n36.Result) {constB=5}
  n45 = Multiply(A=OneMinus(None=Saturate(None=@n44))) {constB=0.2}
  n46 = NamedRerouteUsage()
  n47 = LinearInterpolate(
   A=Desaturation(None=LinearInterpolate(A=@n40, B=@n41.Result, Alpha=@n42), Fraction=Multiply(A=@n43) {constB=-2})
   B=MaterialFunctionCall(Hue Shift Percentage=Multiply(A=@n45, B=@n46) {constB=0.2}, Texture=@n24.RGB) {materialFunction=&HueShift}.Result
   Alpha=@n12.G
  )
  n48 = LinearInterpolate(A=Multiply(A=@n47) {constB=0.2}, B=@n25, Alpha=@n26)
  NamedRerouteDeclaration(None=LinearInterpolate(
    A=VectorParameter() {parameterName=MetalColor, defaultValue=<rgba 0.402 0.4125 0.4125 1>}.RGB
    B=LinearInterpolate(A=@n48, B=@n47, Alpha=@n27)
    Alpha=NamedRerouteUsage()
   )) {name=BaseColor, nodeColor=<rgba 0.3359 0 1 1>}
 ]}
